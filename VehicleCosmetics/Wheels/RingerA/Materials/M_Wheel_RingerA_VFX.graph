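# Translucent unlit VFX material for the RingerA wheel, combining a tinted swirling center circle with lightning in emissive.
# The circle multiplies and sums counter-rotating radial texture samples, masked by an inverted dot texture.
# Lightning multiplies rotated caustics by a lightning texture, with cloud noise switching channels, then masks it with a panned radial curve.
# Static switches, on by default, replace time with scalar parameters to freeze lightning and radial pan.
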Material /VehicleCosmetics/Wheels/RingerA/Materials/M_Wheel_RingerA_VFX
graph {
 properties {
  bAutomaticallySetUsageInEditor=false
  bUseMaterialAttributes=true
  bUsedWithMeshDeformer=true
  bUsedWithSkeletalMesh=true
  blendMode=BLEND_Translucent
  shadingModel=MSM_Unlit
  usedShadingModels=Unlit
 }
 nodes [
  n1 = Clamp(None=ConstantBiasScale(None=TextureSample() {texture=/VehicleCosmetics/Wheels/_SharedTextures/FX/Dot_inverted}.RGB) {bias=-0.2, scale=1.8})
  NamedRerouteDeclaration(None=@n1) {name=Dot, nodeColor=<rgba 1 0 0.9609 1>}
  n2 = Add(A=TextureCoordinate(), B=Constant() {r=-0.5})
  n3 = DotProduct(A=@n2, B=@n2)
  n4 = Multiply(A=Clamp(None=Abs(None=@n3)), B=Constant() {r=2})
  n5 = StaticSwitchParameter(True=ScalarParameter() {defaultValue=0.5, group=Lightning, parameterName=RadialPan}, False=Time()) {defaultValue=true, group=LightningVector, parameterName=RadialPan_Freeze}
  n6 = ComponentMask(None=TextureSample(UVs=Panner(Coordinate=@n4, Time=@n5) {speedX=0.5}) {texture=/VehicleCosmetics/Wheels/_SharedTextures/FX/Curve_Rainbow}.RGB) {g=true}
  n7 = NamedRerouteUsage()
  n8 = StaticSwitchParameter(True=ScalarParameter() {defaultValue=0.6, group=Lightning, parameterName=LightingTime}, False=Time()) {defaultValue=true, group=LightningVector, parameterName=Lighting_Freeze}
  n9 = TextureSample(UVs=Rotator(Coordinate=TextureCoordinate() {uTiling=0.5, vTiling=0.5}, Time=@n8)) {texture=/DealershipSharedLibrary/FX/Textures/Noise/T_VCVFX_CausticsMask_RGB}
  n10 = TextureSample(UVs=Rotator(Coordinate=TextureCoordinate() {uTiling=-0.5, vTiling=0.5})) {texture=&T_VCVFX_CausticsMask_RGB}
  n11 = Multiply(A=Add(A=@n9.R, B=@n10.B), B=Add(A=@n9.B, B=@n10.R))
  n12 = TextureSample(UVs=Panner(Coordinate=TextureCoordinate() {uTiling=1.5, vTiling=1.5}, Time=Constant() {r=0.25}) {speedX=0.5, speedY=-0.5}) {texture=/VehicleCosmetics/Wheels/RingerA/Textures/Lightning_RGB_Random_T}
  n13 = TextureCoordinate() {uTiling=0, vTiling=0}
  n14 = Panner(Coordinate=@n13, Time=@n8) {speedX=1, speedY=0.65}
  n15 = Panner(Coordinate=@n13, Time=@n8) {speedX=-1, speedY=-0.3}
  n16 = AppendVector(
   A=AppendVector(
    A=TextureSample(UVs=@n14) {texture=/VehicleCosmetics/Wheels/_SharedTextures/FX/Clouds_T}.R
    B=TextureSample(UVs=@n15) {texture=&Clouds_T}.R
   )
   B=TextureSample(UVs=Panner(Coordinate=@n13, Time=@n8) {speedX=-0.5, speedY=0.3}) {texture=&Clouds_T}.R
  )
  n17 = Constant() {r=0.2}
  n18 = Clamp(None=Multiply(A=Subtract(A=@n16, B=@n17), B=Constant() {r=35}))
  n19 = LinearInterpolate(A=LinearInterpolate(A=@n12.G, B=@n12.G, Alpha=ComponentMask(None=@n18) {g=true}), B=@n12.B, Alpha=ComponentMask(None=@n18) {b=true})
  n20 = Multiply(
   A=VectorParameter() {group=Lightning, parameterName=LightningColor, defaultValue=<rgba 0.2759 0.7151 1 1>}.RGB
   B=Multiply(A=ScalarParameter() {defaultValue=5, group=Lightning, parameterName=LightningBrightness}, B=Multiply(A=@n11, B=@n19))
  )
  NamedRerouteDeclaration(None=Multiply(A=OneMinus(None=@n6), B=Multiply(A=@n7, B=@n20))) {name=Lightning, nodeColor=<rgba 0.0089 0.6919 1 1>}
  n21 = Rotator() {speed=-1}
  n22 = Rotator(Coordinate=TextureCoordinate() {uTiling=-1}) {centerX=-0.5, speed=0.6}
  n23 = Constant() {r=4}
  n24 = Multiply(
   A=Multiply(
    A=TextureSample(UVs=@n21) {samplerType=SAMPLERTYPE_LinearColor, texture=/VehicleCosmetics/Wheels/RingerA/Textures/T_Ringer_A_Radial02}.R
    B=TextureSample(UVs=@n22) {samplerType=SAMPLERTYPE_LinearColor, texture=&T_Ringer_A_Radial02}.G
   )
   B=@n23
  )
  n25 = VectorParameter() {group=CenterCircle, parameterName=CircleColor, defaultValue=<rgba 0.2396 0.0024 0.0024 1>}
  n26 = Rotator() {speed=1}
  n27 = Rotator() {speed=-0.6}
  n28 = Multiply(
   A=Multiply(
    A=TextureSample(UVs=@n26) {samplerType=SAMPLERTYPE_LinearColor, texture=&T_Ringer_A_Radial02}.R
    B=TextureSample(UVs=@n27) {samplerType=SAMPLERTYPE_LinearColor, texture=&T_Ringer_A_Radial02}.G
   )
   B=@n23
  )
  n29 = Multiply(A=Add(A=Multiply(A=@n24, B=@n25.RGB), B=Multiply(A=@n28, B=@n25.RGB)), B=@n1)
  n30 = ScalarParameter() {defaultValue=0.9, group=CenterCircle, parameterName=EmissiveMultiply}
  SetMaterialAttributes(Emissive Color=Add(A=Multiply(A=@n29, B=@n30), B=NamedRerouteUsage())) {attributeSetTypes [#1]}
 ]
}
